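# Niagara trail for the Ensemble Snake Male glider: one Speedlines sprite emitter with distance culling.
# Spawn rate scales with User.ForwardDot in gameplay, while front-end modifiers swap that and other values for menu presentation.
# Short-lived sprites spawn in a cylinder, get directional and point-outward velocity, stretch with speed and fade by alpha curve.
# Alpha rises as SpawnCountScale drops, compensating for reduced particle counts.
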NiagaraSystem /BRCosmetics/Gadgets/Assets/VinderTech_GliderChute/Glider_Ensemble_Snake_Male/FX/NS_Glider_EnsembleSnakeMale_Trail
dependencies {
 usedDataInterfaces [&NiagaraDataInterfaceCurve, &NiagaraDataInterfaceVector2DCurve]
 usedDynamicInputs []
 usedModules [&SystemState]
}
emitters {Speedlines {
  emitterUpdateScript [
   &EmitterState(Enable Distance Culling=true, MaxDistance=2048, Module.bUseMaxDistance=true, Module.bUseMinDistance=false, Scalability Mode=::NewEnumerator1)
   &SpawnRate(
    SpawnRate=&Multiply_Float(
     A=&FN_FrontEndFloatModifier(Gameplay=$User.ForwardDot, Front End=1)
     B=&UniformRangedFloat(Minimum=75, Maximum=150)
    )
    Use Spawn Probability=false
   )
  ]
  particleSpawnScript [
   &InitializeParticle(
    Color=&MakeLinearColorFromVectorAndFloat(
     "Vector (RGB)"=<1 1 1>
     "Float (Alpha)"=&Multiply_Float(
      A=&Multiply_Float(
       A=&UniformRangedFloat(Minimum=0.015, Maximum=0.1)
       B=&FN_FrontEndFloatModifier(Gameplay=3, Front End=2)
      )
      B=&Add_Float(
       A=1
       B=&Multiply_Float(A=10, B=&Subtract_Float(A=1, B=$Engine.Emitter.SpawnCountScale))
      )
     )
    )
    Lifetime=&FN_FrontEndFloatModifier(
     Gameplay=&RandomRangeFloat(Minimum=0.2, Maximum=0.35)
     Front End=&RandomRangeFloat(Minimum=0.1, Maximum=0.2)
    )
    Sprite Size Max=<5 6>
    Sprite Size Min=<3 4>
    Sprite Size Mode=::NewEnumerator2
    UsePositionOffset=false
    Write Material Random=true
   ) {module=InitializeParticle001}
   &CylinderLocation(
    Cylinder Radius=&FN_FrontEndFloatModifier(
     Gameplay=&UniformRangedFloat(Minimum=100, Maximum=150)
     Front End=&UniformRangedFloat(Minimum=150, Maximum=200)
    )
    MaskSpawnGroup=false
    Override Seed=false
    UseOwnerScale=true
   )
   &AddVelocity(Velocity=&FN_FrontEndVector3Modifier(Gameplay=<0 1000 0>, Front End=<0 3333 0>))
   &AddVelocityFromPoint(
    Module.Use Falloff Distance=false
    Velocity Strength=&Multiply_Float(
     A=&FN_FrontEndFloatModifier(Gameplay=1, Front End=2)
     B=&UniformRangedFloat(Minimum=500, Maximum=700)
    )
   )
  ]
  particleUpdateScript [
   &ParticleState()
   &SolveForcesAndVelocity(Clamp Velocity=false, Limit Acceleration=false)
   &ScaleSpriteSize(Scale Factor=&Vector2DFromCurve(
     Vector2Curve=DataInterface() {propertyValues {XCurve {keys [<at 0 0.5 RCIM_Cubic>,<at 0.5 1 RCIM_Cubic>]}, YCurve {keys [<at 0 0 RCIM_Cubic>,<at 0.5 1 RCIM_Cubic>]}}}
     ScaleCurve=<1 0.35>
    ))
   &ScaleSpriteSizeBySpeed(Max Scale Factor=<1 40>, Min Scale Factor=<0.01 0.05>, Velocity Threshold=1)
   &ScaleColor(
    Scale Alpha=&FloatFromCurve(
     FloatCurve=DataInterface() {propertyValues {Curve {keys [
         <at 0 0 RCIM_Cubic>
         {arriveTangent=-2.2324, interpMode=RCIM_Cubic, leaveTangent=-2.2324, tangentMode=RCTM_User, time=0.2, value=1}
         {arriveTangent=-0.1104, interpMode=RCIM_Cubic, leaveTangent=-0.1104, tangentMode=RCTM_User, time=1, value=0}
        ]}}}
     Scale Curve=&FN_FrontEndFloatModifier(Gameplay=0.4, Front End=1)
    )
    ScaleA=true
    ScaleRGB=true
   )
   &CameraOffset(Camera Offset Amount=&FN_FrontEndFloatModifier(Gameplay=0, Front End=-20))
  ]
  rendererClasses [&NiagaraSpriteRendererProperties]
  renderers [{rendererClass=&NiagaraSpriteRendererProperties}]
 }}
summary {
 systemName=NS_Glider_EnsembleSnakeMale_Trail
 emitters [{emitterName=Speedlines, rendererClasses [&NiagaraSpriteRendererProperties]}]
 userVariables [
  User.bIsFrontEnd {type=NiagaraBool}
  User.bIsFrontEndPreview {type=NiagaraBool}
  User.ForwardDot {type=NiagaraFloat, defaultValue=0, description=""}
 ]
}
